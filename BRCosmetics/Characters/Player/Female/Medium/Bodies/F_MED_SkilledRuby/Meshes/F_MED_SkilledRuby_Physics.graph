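# Physics asset for the F_MED_SkilledRuby female character body, setting up jointed bodies for dangling ring and brush accessories.
# Capsules cover the pelvis, spine, thighs, pockets and gun, with the gun locked to the left thigh and both pockets locked to spine_01.
# A three-segment brush chain hangs from dyn_Ring1 off pocket1's Base_Ring1, while dyn_Ring2 hangs from pocket2, all with angular limits.
# The last two brush segments have no collision shapes.
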PhysicsAsset /BRCosmetics/Characters/Player/Female/Medium/Bodies/F_MED_SkilledRuby/Meshes/F_MED_SkilledRuby_Physics
bodies {
 Base_Ring1 [{shapeType=Capsule, length=0.505, radius=0.505, shapeName=Base_Ring1_capsule, rotation=<rot 0 -90 0>}]
 dyn_01_brush [{
   shapeType=Capsule
   length=3.2029
   radius=1.0526
   shapeName=dyn_01_brush_capsule
   center=<3.1856 0.0117 0.0115>
   rotation=<rot 0 -90.5987 -89.5839>
  }]
 dyn_02_brush []
 dyn_03_brush []
 dyn_Ring1 [{
   shapeType=Capsule
   length=1.6678
   radius=0.851
   shapeName=dyn_Ring1_capsule
   center=<1.3945 0.0086 -0.0167>
   rotation=<rot 0 -11.2844 -87.8889>
  }]
 dyn_Ring2 [{
   shapeType=Capsule
   length=2.1118
   radius=1.77
   shapeName=dyn_Ring2_capsule
   center=<1.4602 0.0613 0.0635>
   rotation=<rot 0 173.5397 -86.7938>
  }]
 gun [{
   shapeType=Capsule
   length=10.8535
   radius=6.2195
   shapeName=gun_capsule
   center=<-0.8206 -0.964 2.8356>
   rotation=<rot -0.0806 -66.011 -89.5755>
  }]
 pelvis [{shapeType=Capsule, length=12.2321, radius=10.1848}]
 pocket1 [{shapeType=Capsule, length=1, radius=1}]
 pocket2 [{
   shapeType=Capsule
   length=6.2561
   radius=3.2588
   shapeName=pocket2_capsule
   center=<2.2509 0.059 -0.3492>
   rotation=<rot 0 -88.6616 -80.1381>
  }]
 spine_01 [{shapeType=Capsule, length=0.505, radius=0.505, shapeName=spine_01_capsule, rotation=<rot 0 -90 0>}]
 spine_02 [{
   shapeType=Capsule
   length=10.3491
   radius=8.5666
   shapeName=spine_02_capsule
   center=<2.5681 1.9123 -0.0721>
   rotation=<rot 0.8529 -90 0.3317>
  }]
 spine_03 [{shapeType=Capsule, length=0.505, radius=0.505, shapeName=spine_03_capsule, rotation=<rot 0 -90 0>}]
 thigh_l [{
   shapeType=Capsule
   length=35.4456
   radius=8.5119
   shapeName=thigh_l_capsule
   center=<-27.7141 -0.3207 -0.2375>
   rotation=<rot 0.896 86.408 -88.987>
  }]
 thigh_r [{
   shapeType=Capsule
   length=14.9077
   radius=8.776
   shapeName=thigh_r_capsule
   center=<27.7141 0.3207 0.2375>
   rotation=<rot 0.9522 89.5657 93.2546>
  }]
}
constraints [
 {
  bone1Name=spine_01
  bone2Name=pelvis
  swing1LimitDegrees=45
  swing1Motion=Free
  swing2LimitDegrees=45
  swing2Motion=Free
  twistLimitDegrees=45
  twistMotion=Free
 }
 {
  bone1Name=spine_02
  bone2Name=spine_01
  swing1LimitDegrees=45
  swing1Motion=Free
  swing2LimitDegrees=45
  swing2Motion=Free
  twistLimitDegrees=45
  twistMotion=Free
 }
 {
  bone1Name=spine_03
  bone2Name=spine_02
  swing1LimitDegrees=45
  swing1Motion=Free
  swing2LimitDegrees=45
  swing2Motion=Free
  twistLimitDegrees=45
  twistMotion=Free
 }
 {
  bone1Name=pocket1
  bone2Name=spine_01
  swing1LimitDegrees=45
  swing1Motion=Locked
  swing2LimitDegrees=45
  swing2Motion=Locked
  twistLimitDegrees=45
  twistMotion=Locked
 }
 {bone1Name=dyn_Ring1, bone2Name=Base_Ring1, swing1LimitDegrees=1, swing2LimitDegrees=15, twistLimitDegrees=1, twistMotion=Locked}
 {bone1Name=dyn_01_brush, bone2Name=dyn_Ring1, swing1LimitDegrees=30, swing2LimitDegrees=30, twistLimitDegrees=1}
 {bone1Name=dyn_02_brush, bone2Name=dyn_01_brush, swing1LimitDegrees=33, swing2LimitDegrees=33, twistLimitDegrees=1}
 {bone1Name=dyn_03_brush, bone2Name=dyn_02_brush, swing1LimitDegrees=33, swing2LimitDegrees=33, twistLimitDegrees=1}
 {
  bone1Name=thigh_l
  bone2Name=pelvis
  swing1LimitDegrees=45
  swing1Motion=Free
  swing2LimitDegrees=45
  swing2Motion=Free
  twistLimitDegrees=45
  twistMotion=Free
 }
 {
  bone1Name=thigh_r
  bone2Name=pelvis
  swing1LimitDegrees=45
  swing1Motion=Free
  swing2LimitDegrees=45
  swing2Motion=Free
  twistLimitDegrees=45
  twistMotion=Free
 }
 {
  bone1Name=pocket2
  bone2Name=spine_01
  swing1LimitDegrees=45
  swing1Motion=Locked
  swing2LimitDegrees=45
  swing2Motion=Locked
  twistLimitDegrees=45
  twistMotion=Locked
 }
 {bone1Name=dyn_Ring2, bone2Name=pocket2, swing1LimitDegrees=1, swing2LimitDegrees=5, twistLimitDegrees=1}
 {
  bone1Name=Base_Ring1
  bone2Name=pocket1
  swing1LimitDegrees=45
  swing1Motion=Locked
  swing2LimitDegrees=45
  swing2Motion=Locked
  twistLimitDegrees=45
  twistMotion=Locked
 }
 {
  bone1Name=gun
  bone2Name=thigh_l
  swing1LimitDegrees=45
  swing1Motion=Locked
  swing2LimitDegrees=45
  swing2Motion=Locked
  twistLimitDegrees=45
  twistMotion=Locked
 }
]
