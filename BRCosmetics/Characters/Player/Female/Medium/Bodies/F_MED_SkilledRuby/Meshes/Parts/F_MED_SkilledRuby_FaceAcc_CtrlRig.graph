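# Control Rig for the SkilledRuby female face-accessory part, applying corrective offsets to hair and accessory bones.
# Head rotation relative to spine_05 weights hair corrections, followed by a constant Base_bu offset.
# Two multiplied pelvis-height remaps form a height-window weight for a further correction.
# The isEmote flag enables another head-driven hair correction, while the isSky branch and some duplicate nodes sit outside the execution chain.
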ControlRigBlueprint /BRCosmetics/Characters/Player/Female/Medium/Bodies/F_MED_SkilledRuby/Meshes/Parts/F_MED_SkilledRuby_FaceAcc_CtrlRig
graphs [
 RigVMFunctionLibrary {links [], nodes []}
 RigVMModel {nodes [
   RigUnit_BeginExecution = &RigUnit_BeginExecution()
   n1 = &RigVMFunction_MathDoubleRemap(
    SourceMaximum=-45
    SourceMinimum=0
    TargetMaximum=1
    TargetMinimum=0
    Value=&RigVMFunction_MathQuaternionToEuler(
     RotationOrder=XYZ
     Value=&RigUnit_GetRelativeTransformForItem(Child=(Type=Bone,Name="head"), Parent=(Type=Bone,Name="spine_05"), bChildInitial=False, bParentInitial=False).RelativeTransform.Rotation
    ).Result.Z
   )
   &RigUnit_ModifyTransforms(
    ItemToModify="((Item=(Type=Bone,Name=\"Base_01_hair_r\"),Transform=(Rotation=(X=0,Y=-0.7001,Z=0,W=0.714),Translation=(X=-2.1746,Y=0,Z=0),Scale3D=(X=1,Y=1,Z=1))),(Item=(Type=Bone,Name=\"Base_01_hair_mu\"),Transform=(Rotation=(X=0,Y=0,Z=0,W=1),Translation=(X=1.4663,Y=0,Z=0),S ...(truncated)"
    WeightMinimum=0
    Weight=@n1.Result
    ExecutePin=@RigUnit_BeginExecution.ExecutePin
   )
   > &RigUnit_ModifyTransforms(
    ItemToModify=((Item=(Type=Bone,Name="Base_01_hair_mu"),Transform=(Rotation=(X=0.0911,Y=0.6023,Z=-0.1186,W=-0.7842),Translation=(X=-3.4453,Y=0.0001,Z=1.5569),Scale3D=(X=1,Y=1,Z=1))))
    WeightMinimum=0
    Weight=@n1.Result
   )
   > &RigUnit_ModifyTransforms(
    ItemToModify="((Item=(Type=Bone,Name=\"Base_bu_l\"),Transform=(Rotation=(X=0,Y=0,Z=0.2494,W=0.9684),Translation=(X=0,Y=0,Z=0),Scale3D=(X=1,Y=1,Z=1))),(Item=(Type=Bone,Name=\"Base_bu_r\"),Transform=(Rotation=(X=0,Y=0,Z=-0.2032,W=0.9791),Translation=(X=0,Y=0,Z=0),Scale3D=(X=1 ...(truncated)"
    Weight=1
    WeightMinimum=0
   )
   n2 = &RigUnit_GetRelativeTransformForItem(Child=(Type=Bone,Name="pelvis"), Parent=(Type=Bone,Name="root"), bChildInitial=False, bParentInitial=False)
   > n3 = &RigUnit_ModifyTransforms(
    ItemToModify="((Item=(Type=Bone,Name=\"Base_bu_l\"),Transform=(Rotation=(X=0,Y=0,Z=0,W=1),Translation=(X=0,Y=0,Z=0),Scale3D=(X=1,Y=1,Z=1))),(Item=(Type=Bone,Name=\"Base_bu_r\"),Transform=(Rotation=(X=0,Y=0.2401,Z=0,W=-0.9707),Translation=(X=0,Y=0,Z=0),Scale3D=(X=1,Y=1,Z=1)) ...(truncated)"
    WeightMinimum=0
    Weight=&RigVMFunction_MathDoubleMul(
     B=&RigVMFunction_MathDoubleRemap(SourceMaximum=67, SourceMinimum=90, TargetMaximum=1, TargetMinimum=0, Value=@n2.RelativeTransform.Translation.Z).Result
     A=&RigVMFunction_MathDoubleRemap(SourceMaximum=69, SourceMinimum=50, TargetMaximum=1, TargetMinimum=0, Value=@n2.RelativeTransform.Translation.Z).Result
    ).Result
   )
   n4 = &RigVMFunction_ControlFlowBranch(ExecuteContext=@n3.ExecutePin, Condition=RigVMVariableNode() {variable=isEmote}.Value)
   n5 = &RigUnit_GetRelativeTransformForItem(Child=(Type=Bone,Name="head"), Parent=(Type=Bone,Name="spine_05"), bChildInitial=False, bParentInitial=False)
   &RigUnit_ModifyTransforms(
    ItemToModify="((Item=(Type=Bone,Name=\"Base_01_hair_mu\"),Transform=(Rotation=(X=0,Y=-0.3827,Z=0,W=0.9239),Translation=(X=1,Y=0,Z=1.5),Scale3D=(X=1,Y=1,Z=1))),(Item=(Type=Bone,Name=\"Base_01_hair_mb\"),Transform=(Rotation=(X=0,Y=0,Z=0,W=1),Translation=(X=0.244,Y=0,Z=1.4),Sc ...(truncated)"
    WeightMinimum=0
    Weight=&RigVMFunction_MathDoubleRemap(
     SourceMaximum=-45
     SourceMinimum=0
     TargetMaximum=1
     TargetMinimum=0
     Value=&RigVMFunction_MathQuaternionToEuler(RotationOrder=XYZ, Value=@n5.RelativeTransform.Rotation).Result.Z
    ).Result
    ExecutePin=@n4.True
   )
   &RigUnit_ModifyTransforms(
    ExecutePin=()
    ItemToModify="((Item=(Type=Bone,Name=\"Base_01_hair_r\"),Transform=(Rotation=(X=0.007,Y=0.7071,Z=-0.007,W=-0.7071),Translation=(X=-2.1746,Y=0,Z=0),Scale3D=(X=1,Y=1,Z=1))),(Item=(Type=Bone,Name=\"Base_01_hair_mu\"),Transform=(Rotation=(X=0,Y=0,Z=0,W=1),Translation=(X=1.4663, ...(truncated)"
    Weight=1
    WeightMinimum=0
   )
   &RigVMFunction_MathDoubleRemap(
    SourceMaximum=-45
    SourceMinimum=0
    TargetMaximum=1
    TargetMinimum=0
    Value=&RigVMFunction_MathQuaternionToEuler(
     RotationOrder=XYZ
     Value=&RigUnit_GetRelativeTransformForItem(Child=(Type=Bone,Name="head"), Parent=(Type=Bone,Name="spine_05"), bChildInitial=False, bParentInitial=False).RelativeTransform.Rotation
    ).Result.Z
   )
   RigVMFunction_ControlFlowBranch = &RigVMFunction_ControlFlowBranch(ExecuteContext=(), Condition=RigVMVariableNode() {variable=isSky}.Value)
  ]}
]
hierarchy [
 root {loc=<0 0 0>, rot=<0 0 0 1>}
 pelvis {parent=root, loc=<0 -0.5301 98.0571>, rot=<0.0127 0.707 -0.0127 -0.707>}
 spine_01 {parent=pelvis, loc=<6.1208 -0.9328 0>, rot=<0 0 0.034 -0.9994>}
 spine_02 {parent=spine_01, loc=<6.5997 -0.2095 0>, rot=<0 0 0.0238 -0.9997>}
 spine_03 {parent=spine_02, loc=<6.4914 -0.1237 0>, rot=<0 0 0.0149 -0.9999>}
 spine_04 {parent=spine_03, loc=<6.9598 0.0794 0>, rot=<0 0 -0.0132 -0.9999>}
 spine_05 {parent=spine_04, loc=<6.9987 0 0>, rot=<0 0 -0.0015 -1>}
 neck_01 {parent=spine_05, loc=<11.3138 0.1529 0>, rot=<0 0 -0.2003 -0.9797>}
 neck_02 {parent=neck_01, loc=<4.5627 0 0>, rot=<0 0 0.0504 -0.9987>}
 head {parent=neck_02, loc=<5.3914 0 0>, rot=<0 0 0.115 -0.9934>}
 Base_01_hair_r {parent=head, loc=<0.7477 -3.6734 3.9882>, rot=<0.1227 -0.5986 0.7916 -0.0048>}
 dyn_01_hair_r {parent=Base_01_hair_r, loc=<0 0 0>, rot=<0 0 0 1>}
 dyn_02_hair_r {parent=dyn_01_hair_r, loc=<5.3476 0 0.144>, rot=<0 0.1288 0 -0.9917>}
 dyn_03_hair_r {parent=dyn_02_hair_r, loc=<5.0648 0 0>, rot=<0 0 0 1>}
 Base_01_hair_mu {parent=head, loc=<1.1478 -7.2394 -0.0407>, rot=<0.0658 -0.7059 0.704 -0.0408>}
 dyn_01_hair_mu {parent=Base_01_hair_mu, loc=<0 0 0>, rot=<0 0 0 1>}
 dyn_02_hair_mu {parent=dyn_01_hair_mu, loc=<4.5157 0 0>, rot=<0 0.0856 0 -0.9963>}
 dyn_03_hair_mu {parent=dyn_02_hair_mu, loc=<3.5152 0 0>, rot=<0 0.0466 0 -0.9989>}
 dyn_04_hair_mu {parent=dyn_03_hair_mu, loc=<2.9225 0 0>, rot=<0 0.0511 0 -0.9987>}
 dyn_05_hair_mu {parent=dyn_04_hair_mu, loc=<3.4008 0 0>, rot=<0 0 0 1>}
 Base_bu_l {parent=Base_01_hair_mu, loc=<0.2135 0.822 0>, rot=<0 0 -0.2879 -0.9577>}
 dyn_bu_l {parent=Base_bu_l, loc=<0 0 0>, rot=<0 0 0 1>}
 Base_bu_r {parent=Base_01_hair_mu, loc=<0 -0.9621 0>, rot=<0 0 0.4261 -0.9047>}
 dyn_bu_r {parent=Base_bu_r, loc=<0 0 0>, rot=<0 0 0 1>}
 dyn_bu_r2 {parent=dyn_bu_r, loc=<2.3112 0 -0.3816>, rot=<0 0.0856 0 -0.9963>}
 Base_01_hair_l {parent=head, loc=<0.4496 -3.918 -4.4792>, rot=<0.075 -0.9091 0.4065 -0.0506>}
 dyn_01_hair_l {parent=Base_01_hair_l, loc=<0 0 0>, rot=<0 0 0 1>}
 dyn_02_hair_l {parent=dyn_01_hair_l, loc=<5.3476 0 0.144>, rot=<0 0.1288 0 -0.9917>}
 dyn_03_hair_l {parent=dyn_02_hair_l, loc=<5.0648 0 0>, rot=<0 0 0 1>}
 Base_01_hair_mb {parent=head, loc=<0.8565 -5.3239 -0.0304>, rot=<0.0658 -0.7059 0.704 -0.0408>}
 dyn_01_hair_mb {parent=Base_01_hair_mb, loc=<0 0 0>, rot=<0 0 0 1>}
 dyn_02_hair_mb {parent=dyn_01_hair_mb, loc=<4.5157 0 0>, rot=<0 0.0856 0 -0.9963>}
 dyn_03_hair_mb {parent=dyn_02_hair_mb, loc=<5.9123 0 0>, rot=<0 0.1086 0 -0.9941>}
 dyn_04_hair_mb {parent=dyn_03_hair_mb, loc=<5.7787 0 0>, rot=<0 0 0 1>}
 Bang_01_l {parent=head, loc=<8.6138 8.4907 -4.0359>, rot=<0.0184 -0.8649 -0.5011 -0.0236>}
 Bang_02_l {parent=Bang_01_l, loc=<1.7077 0 0>, rot=<0 0 0 1>}
 Bang_01_m {parent=head, loc=<7.4097 9.5024 -1.1037>, rot=<0.0228 -0.7458 -0.6655 -0.0194>}
 Bang_02_m {parent=head, loc=<7.7249 9.1941 1.4433>, rot=<-0.0249 0.6638 0.7473 0.0166>}
 Bang_01_r {parent=head, loc=<7.4676 7.6951 4.2408>, rot=<-0.0291 0.3642 0.9308 0.007>}
 hair_corrHead {parent=head, loc=<0.0111 -1.3474 0>, rot=<0.0029 -0.7071 -0.0029 -0.7071>}
 neck_skale {parent=neck_01, loc=<0 0 0>, rot=<0 0 0 1>}
 clavicle_l {parent=spine_05, loc=<5.4831 -0.1053 -3.2946>, rot=<0.0465 -0.7289 -0.009 -0.6829>}
 upperarm_l {parent=clavicle_l, loc=<9.8658 -1.2136 -0.4105>, rot=<0.0164 -0.3245 -0.0269 -0.9454>}
 clavicle_r {parent=spine_05, loc=<5.4831 -0.1053 3.2946>, rot=<0.7289 0.0465 -0.6829 0.009>}
 upperarm_r {parent=clavicle_r, loc=<-9.8658 1.2136 0.4105>, rot=<0.0164 -0.3245 -0.0269 -0.9454>}
 attach_backpack {parent=spine_05, loc=<-0.123 -13.473 0>, rot=<-0.0093 -0.707 0.0093 -0.707>}
 attach_cape {parent=spine_05, loc=<9.017 -11.714 0>, rot=<0.0616 -0.7044 -0.0616 -0.7044>}
]
variables [
 isEmote {type=bool, default=False}
 isSky {type=bool, default=False}
]
