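# Physics asset for the SkilledRuby female medium face accessory mesh, defining collision bodies and joints for its simulated hair.
# Upper spine, clavicle, neck, head and hair base bodies are joined by fully locked constraints.
# Left, right and middle-back dyn hair chains hang from their bases with swing and twist limits, the middle adding boxes to capsules.
# Tight cross-constraints tie the last middle bone to both side-chain ends, linking the three strands.
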PhysicsAsset /BRCosmetics/Characters/Player/Female/Medium/Bodies/F_MED_SkilledRuby/Meshes/Parts/F_MED_SkilledRuby_FaceAcc_Physics
bodies {
 Base_01_hair_l [{shapeType=Capsule, length=0.505, radius=0.505, shapeName=Base_01_hair_l_capsule, rotation=<rot 0 -90 0>}]
 Base_01_hair_mb [{shapeType=Capsule, length=0.505, radius=0.505, shapeName=Base_01_hair_mb_capsule, rotation=<rot 0 -90 0>}]
 Base_01_hair_mu [{shapeType=Capsule, length=0.505, radius=0.505, shapeName=Base_01_hair_mu_capsule, rotation=<rot 0 -90 0>}]
 Base_01_hair_r [{shapeType=Capsule, length=0.505, radius=0.505, shapeName=Base_01_hair_r_capsule, rotation=<rot 0 -90 0>}]
 clavicle_l [{
   shapeType=Capsule
   length=16.6347
   radius=5.2445
   shapeName=clavicle_l_capsule
   center=<8.1424 0 -0.2789>
   rotation=<rot -0.0001 -90 -88.038>
  }]
 clavicle_r [{shapeType=Capsule, length=17.0923, radius=4.7869, shapeName=clavicle_r_capsule, center=<-8.1424 0 0.2789>, rotation=<rot -0.0001 -90 91.962>}]
 dyn_01_hair_l [{shapeType=Capsule, length=2.5087, radius=1.8511, center=<1.4953 0 0.3>, rotation=<rot 0 0 90>}]
 dyn_01_hair_mb [{shapeType=Capsule, length=0.3285, radius=2.1454, rotation=<rot 0 0 90>}]
 dyn_01_hair_r [{shapeType=Capsule, length=2.5087, radius=1.8511, center=<1.4953 0 0.3>, rotation=<rot 0 0 90>}]
 dyn_02_hair_l [{shapeType=Capsule, length=5.1617, radius=3.0656, center=<2.0265 0 1.2>, rotation=<rot 84.8853 -179.9995 -179.9996>}]
 dyn_02_hair_mb [
  {shapeType=Capsule, length=1, radius=1, center=<0.3 0 0>, rotation=<rot 0 0 90>}
  {shapeType=Box, extentX=5.0093, extentY=10.0577, extentZ=2.012, shapeName=dyn_02_hair_mb_box, center=<2.7827 0 0>}
 ]
 dyn_02_hair_r [{
   shapeType=Capsule
   length=5.1747
   radius=3.0526
   shapeName=dyn_02_hair_r_capsule
   center=<3.2568 0 1.3>
   rotation=<rot -8.6495 -97.9498 -86.2618>
  }]
 dyn_03_hair_l [{shapeType=Capsule, length=4.1314, radius=3.1535, center=<1.6146 0 2>, rotation=<rot 10.0898 22.9849 -103.6278>}]
 dyn_03_hair_mb [
  {shapeType=Capsule, length=1, radius=1, center=<0.3 0 0>, rotation=<rot 0 0 90>}
  {shapeType=Box, extentX=5.2869, extentY=8.9422, extentZ=2.1217, shapeName=dyn_03_hair_mb_box, center=<2.3289 0 0>}
 ]
 dyn_03_hair_r [{
   shapeType=Capsule
   length=2.1542
   radius=3.6315
   shapeName=dyn_03_hair_r_capsule
   center=<2.125 0 1.8>
   rotation=<rot -0.5855 158.5415 -79.4474>
  }]
 dyn_04_hair_mb [
  {shapeType=Capsule, length=1, radius=1, center=<0.3 0 0>, rotation=<rot 0 0 90>}
  {shapeType=Box, extentX=5.754, extentY=8.9422, extentZ=2.4555, shapeName=dyn_04_hair_mb_box, center=<2.3494 0 0>}
 ]
 head [{shapeType=Capsule, length=1, radius=1}]
 neck_01 [{shapeType=Capsule, length=4.5435, radius=5.0963, shapeName=neck_01_capsule, center=<0 1.6924 0>, rotation=<rot 0 -90 90>}]
 neck_02 [{shapeType=Capsule, length=1.8485, radius=4.9241, shapeName=neck_02_capsule, center=<0 1.9996 0>, rotation=<rot 0 -90 90>}]
 spine_05 [{shapeType=Capsule, length=13.2398, radius=9.2278, shapeName=spine_05_capsule, center=<1.5672 0.5337 0>, rotation=<rot 0 -90.0001 0>}]
}
constraints [
 {bone1Name=dyn_01_hair_l, bone2Name=Base_01_hair_l, swing1LimitDegrees=35, swing2LimitDegrees=40, twistLimitDegrees=20}
 {bone1Name=dyn_02_hair_l, bone2Name=dyn_01_hair_l, swing1LimitDegrees=25, swing2LimitDegrees=25, twistLimitDegrees=20}
 {bone1Name=dyn_03_hair_l, bone2Name=dyn_02_hair_l, swing1LimitDegrees=25, swing2LimitDegrees=25, twistLimitDegrees=1}
 {bone1Name=dyn_01_hair_mb, bone2Name=Base_01_hair_mb, swing1LimitDegrees=20, swing2LimitDegrees=30, twistLimitDegrees=5}
 {bone1Name=dyn_02_hair_mb, bone2Name=dyn_01_hair_mb, swing1LimitDegrees=20, swing2LimitDegrees=20, twistLimitDegrees=20}
 {bone1Name=dyn_03_hair_mb, bone2Name=dyn_02_hair_mb, swing1LimitDegrees=20, swing2LimitDegrees=30, twistLimitDegrees=5}
 {bone1Name=dyn_04_hair_mb, bone2Name=dyn_03_hair_mb, swing1LimitDegrees=35, swing2LimitDegrees=35, twistLimitDegrees=5}
 {bone1Name=dyn_03_hair_r, bone2Name=dyn_02_hair_r, swing1LimitDegrees=25, swing2LimitDegrees=25, twistLimitDegrees=1}
 {bone1Name=dyn_02_hair_r, bone2Name=dyn_01_hair_r, swing1LimitDegrees=25, swing2LimitDegrees=25, twistLimitDegrees=20}
 {bone1Name=dyn_01_hair_r, bone2Name=Base_01_hair_r, swing1LimitDegrees=20, swing2LimitDegrees=40, twistLimitDegrees=20}
 {
  bone1Name=clavicle_l
  bone2Name=spine_05
  swing1LimitDegrees=45
  swing1Motion=Locked
  swing2LimitDegrees=45
  swing2Motion=Locked
  twistLimitDegrees=45
  twistMotion=Locked
 }
 {
  bone1Name=clavicle_r
  bone2Name=spine_05
  swing1LimitDegrees=45
  swing1Motion=Locked
  swing2LimitDegrees=45
  swing2Motion=Locked
  twistLimitDegrees=45
  twistMotion=Locked
 }
 {
  bone1Name=neck_02
  bone2Name=neck_01
  swing1LimitDegrees=45
  swing1Motion=Locked
  swing2LimitDegrees=45
  swing2Motion=Locked
  twistLimitDegrees=45
  twistMotion=Locked
 }
 {
  bone1Name=neck_01
  bone2Name=spine_05
  swing1LimitDegrees=45
  swing1Motion=Locked
  swing2LimitDegrees=45
  swing2Motion=Locked
  twistLimitDegrees=45
  twistMotion=Locked
 }
 {
  bone1Name=Base_01_hair_r
  bone2Name=head
  swing1LimitDegrees=45
  swing1Motion=Locked
  swing2LimitDegrees=45
  swing2Motion=Locked
  twistLimitDegrees=45
  twistMotion=Locked
 }
 {
  bone1Name=Base_01_hair_mb
  bone2Name=head
  swing1LimitDegrees=45
  swing1Motion=Locked
  swing2LimitDegrees=45
  swing2Motion=Locked
  twistLimitDegrees=45
  twistMotion=Locked
 }
 {
  bone1Name=Base_01_hair_l
  bone2Name=head
  swing1LimitDegrees=45
  swing1Motion=Locked
  swing2LimitDegrees=45
  swing2Motion=Locked
  twistLimitDegrees=45
  twistMotion=Locked
 }
 {bone1Name=dyn_04_hair_mb, bone2Name=dyn_03_hair_l, swing1LimitDegrees=1, swing2LimitDegrees=1, twistLimitDegrees=5}
 {bone1Name=dyn_04_hair_mb, bone2Name=dyn_03_hair_r, swing1LimitDegrees=1, swing2LimitDegrees=1, twistLimitDegrees=5}
 {
  bone1Name=Base_01_hair_mu
  bone2Name=head
  swing1LimitDegrees=45
  swing1Motion=Locked
  swing2LimitDegrees=45
  swing2Motion=Locked
  twistLimitDegrees=45
  twistMotion=Locked
 }
 {
  bone1Name=head
  bone2Name=neck_02
  swing1LimitDegrees=45
  swing1Motion=Locked
  swing2LimitDegrees=45
  swing2Motion=Locked
  twistLimitDegrees=45
  twistMotion=Locked
 }
]
