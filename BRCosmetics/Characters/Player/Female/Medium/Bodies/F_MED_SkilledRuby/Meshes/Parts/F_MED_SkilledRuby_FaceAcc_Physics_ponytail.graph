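# Physics asset for the SkilledRuby female character's ponytail face-accessory mesh, defining collision bodies and joints for its hair bones.
# Left, right and mb hair chains use mostly box bodies with angularly locked joints, rooted at spine_05.
# The capsule-based mu chain hangs off the mb base with limited swing and tight twist.
# Small left and right bu capsule chains branch from the mu base with narrower limits; the right has an extra segment.
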PhysicsAsset /BRCosmetics/Characters/Player/Female/Medium/Bodies/F_MED_SkilledRuby/Meshes/Parts/F_MED_SkilledRuby_FaceAcc_Physics_ponytail
bodies {
 Base_01_hair_l [{shapeType=Box, extentX=5.7379, extentY=9.4023, extentZ=1.7603, center=<1.8956 0 0>, rotation=<rot -0.0001 0.0001 -28.431>}]
 Base_01_hair_mb [{shapeType=Capsule, length=1.1777, radius=0.505, shapeName=Base_01_hair_mb_capsule, center=<0 0 -0.9928>, rotation=<rot 0 -90 0>}]
 Base_01_hair_mu [{shapeType=Capsule, length=0.505, radius=0.505, shapeName=Base_01_hair_mu_capsule, rotation=<rot 0 -90 0>}]
 Base_01_hair_r [{shapeType=Box, extentX=5.7379, extentY=9.4023, extentZ=1.7603, center=<1.8956 0 0>}]
 Base_bu_l [{shapeType=Capsule, length=0, radius=0, shapeName=Base_bu_l_capsule}]
 Base_bu_r [{shapeType=Capsule, length=0, radius=0}]
 dyn_01_hair_l [{
   shapeType=Box
   extentX=2.1855
   extentY=0.4971
   extentZ=3.2744
   shapeName=dyn_01_hair_l_box
   center=<2.8148 -1.2341 0.4658>
   rotation=<rot 0 -101.9125 -84.3788>
  }]
 dyn_01_hair_mb [{shapeType=Box, extentX=8.3318, extentY=9.4023, extentZ=2.6588, center=<1.8956 0 -1.1139>}]
 dyn_01_hair_mu [{shapeType=Capsule, length=1.89, radius=0.8057, center=<1.0878 0 0>, rotation=<rot 89.4687 180 180>}]
 dyn_01_hair_r [{shapeType=Box, extentX=5.7379, extentY=9.4023, extentZ=1.7603, center=<1.8956 0 0>}]
 dyn_02_hair_l [{shapeType=Box, extentX=5.7379, extentY=9.4023, extentZ=1.7603, center=<1.8956 0 0>, rotation=<rot -0.0001 0.0001 -20.7425>}]
 dyn_02_hair_mb [{shapeType=Box, extentX=8.3318, extentY=9.4023, extentZ=3.1057, center=<1.8956 0 -0.9928>}]
 dyn_02_hair_mu [{shapeType=Capsule, length=2.7105, radius=1.0526, center=<0.0168 0 0.2746>, rotation=<rot 86.2903 0 0>}]
 dyn_02_hair_r [{shapeType=Box, extentX=5.7379, extentY=9.4023, extentZ=1.7603, center=<1.8956 0 0>}]
 dyn_03_hair_l [{shapeType=Box, extentX=7.0275, extentY=9.4023, extentZ=1.7603, center=<1.8956 0 0>, rotation=<rot -0.0001 0.0001 -20.7425>}]
 dyn_03_hair_mb [{shapeType=Box, extentX=8.3318, extentY=9.4023, extentZ=4.2049, center=<1.8956 0 -1.4552>, rotation=<rot -2.4935 0 0>}]
 dyn_03_hair_mu [{shapeType=Capsule, length=1.713, radius=1.2496, center=<1.0878 0 0>, rotation=<rot 90 1.8571 1.8571>}]
 dyn_03_hair_r [{shapeType=Box, extentX=7.0275, extentY=9.4023, extentZ=1.7603, center=<1.8956 0 0>}]
 dyn_04_hair_mb [{shapeType=Box, extentX=8.3318, extentY=9.4023, extentZ=3.1057, center=<1.8956 0 -0.9928>}]
 dyn_04_hair_mu [{shapeType=Capsule, length=1.5677, radius=1.128, center=<1.0878 0 0>, rotation=<rot 90 1.7638 1.7638>}]
 dyn_05_hair_mu [{shapeType=Capsule, length=3.7167, radius=1.4343, center=<1.0878 0 0>, rotation=<rot 90 1.7638 1.7638>}]
 dyn_bu_l [{shapeType=Capsule, length=1.2219, radius=1.0622, shapeName=dyn_bu_l_capsule, center=<1.8959 0 0>, rotation=<rot 0 -4.1396 -73.2729>}]
 dyn_bu_r [{shapeType=Capsule, length=1.3342, radius=0.7191, shapeName=dyn_bu_r_capsule, center=<1.3262 0 0>, rotation=<rot 0 174.1657 -81.6799>}]
 dyn_bu_r2 [{shapeType=Capsule, length=1.548, radius=1.2043, shapeName=dyn_bu_r2_capsule, center=<0.9586 0 0>, rotation=<rot 0 174.5769 -80.4645>}]
 spine_05 [{shapeType=Capsule, length=0.505, radius=0.505, shapeName=spine_05_capsule, rotation=<rot 0 -90 0>}]
}
constraints [
 {
  bone1Name=dyn_01_hair_mb
  bone2Name=Base_01_hair_mb
  swing1LimitDegrees=45
  swing1Motion=Locked
  swing2LimitDegrees=45
  swing2Motion=Locked
  twistLimitDegrees=45
  twistMotion=Locked
 }
 {
  bone1Name=dyn_02_hair_mb
  bone2Name=dyn_01_hair_mb
  swing1LimitDegrees=45
  swing1Motion=Locked
  swing2LimitDegrees=45
  swing2Motion=Locked
  twistLimitDegrees=45
  twistMotion=Locked
 }
 {
  bone1Name=dyn_03_hair_mb
  bone2Name=dyn_02_hair_mb
  swing1LimitDegrees=45
  swing1Motion=Locked
  swing2LimitDegrees=45
  swing2Motion=Locked
  twistLimitDegrees=45
  twistMotion=Locked
 }
 {
  bone1Name=dyn_04_hair_mb
  bone2Name=dyn_03_hair_mb
  swing1LimitDegrees=45
  swing1Motion=Locked
  swing2LimitDegrees=45
  swing2Motion=Locked
  twistLimitDegrees=45
  twistMotion=Locked
 }
 {
  bone1Name=Base_01_hair_r
  bone2Name=spine_05
  swing1LimitDegrees=45
  swing1Motion=Locked
  swing2LimitDegrees=45
  swing2Motion=Locked
  twistLimitDegrees=45
  twistMotion=Locked
 }
 {
  bone1Name=dyn_01_hair_r
  bone2Name=Base_01_hair_r
  swing1LimitDegrees=45
  swing1Motion=Locked
  swing2LimitDegrees=45
  swing2Motion=Locked
  twistLimitDegrees=45
  twistMotion=Locked
 }
 {
  bone1Name=dyn_02_hair_r
  bone2Name=dyn_01_hair_r
  swing1LimitDegrees=45
  swing1Motion=Locked
  swing2LimitDegrees=45
  swing2Motion=Locked
  twistLimitDegrees=45
  twistMotion=Locked
 }
 {
  bone1Name=dyn_03_hair_r
  bone2Name=dyn_02_hair_r
  swing1LimitDegrees=45
  swing1Motion=Locked
  swing2LimitDegrees=45
  swing2Motion=Locked
  twistLimitDegrees=45
  twistMotion=Locked
 }
 {bone1Name=dyn_01_hair_mu, bone2Name=Base_01_hair_mu, swing1LimitDegrees=30, swing2LimitDegrees=30, twistLimitDegrees=5}
 {bone1Name=dyn_02_hair_mu, bone2Name=dyn_01_hair_mu, swing1LimitDegrees=30, swing2LimitDegrees=30, twistLimitDegrees=5}
 {bone1Name=dyn_03_hair_mu, bone2Name=dyn_02_hair_mu, swing1LimitDegrees=30, swing2LimitDegrees=30, twistLimitDegrees=5}
 {bone1Name=dyn_04_hair_mu, bone2Name=dyn_03_hair_mu, swing1LimitDegrees=30, swing2LimitDegrees=30, twistLimitDegrees=5}
 {bone1Name=dyn_05_hair_mu, bone2Name=dyn_04_hair_mu, swing1LimitDegrees=30, swing2LimitDegrees=30, twistLimitDegrees=5}
 {bone1Name=dyn_bu_r2, bone2Name=dyn_bu_r, swing1LimitDegrees=5, swing2LimitDegrees=20, twistLimitDegrees=1}
 {
  bone1Name=Base_bu_l
  bone2Name=Base_01_hair_mu
  swing1LimitDegrees=45
  swing1Motion=Locked
  swing2LimitDegrees=45
  swing2Motion=Locked
  twistLimitDegrees=45
  twistMotion=Locked
 }
 {bone1Name=dyn_bu_l, bone2Name=Base_bu_l, swing1LimitDegrees=5, swing2LimitDegrees=20, twistLimitDegrees=1}
 {
  bone1Name=Base_bu_r
  bone2Name=Base_01_hair_mu
  swing1LimitDegrees=45
  swing1Motion=Locked
  swing2LimitDegrees=45
  swing2Motion=Locked
  twistLimitDegrees=45
  twistMotion=Locked
 }
 {bone1Name=dyn_bu_r, bone2Name=Base_bu_r, swing1LimitDegrees=5, swing2LimitDegrees=15, twistLimitDegrees=1}
 {
  bone1Name=Base_01_hair_mu
  bone2Name=Base_01_hair_mb
  swing1LimitDegrees=45
  swing1Motion=Locked
  swing2LimitDegrees=45
  swing2Motion=Locked
  twistLimitDegrees=45
  twistMotion=Locked
 }
 {
  bone1Name=Base_01_hair_mb
  bone2Name=spine_05
  swing1LimitDegrees=45
  swing1Motion=Locked
  swing2LimitDegrees=45
  swing2Motion=Locked
  twistLimitDegrees=45
  twistMotion=Locked
 }
 {
  bone1Name=dyn_03_hair_l
  bone2Name=dyn_02_hair_l
  swing1LimitDegrees=45
  swing1Motion=Locked
  swing2LimitDegrees=45
  swing2Motion=Locked
  twistLimitDegrees=45
  twistMotion=Locked
 }
 {
  bone1Name=Base_01_hair_l
  bone2Name=spine_05
  swing1LimitDegrees=45
  swing1Motion=Locked
  swing2LimitDegrees=45
  swing2Motion=Locked
  twistLimitDegrees=45
  twistMotion=Locked
 }
 {
  bone1Name=dyn_02_hair_l
  bone2Name=dyn_01_hair_l
  swing1LimitDegrees=45
  swing1Motion=Locked
  swing2LimitDegrees=45
  swing2Motion=Locked
  twistLimitDegrees=45
  twistMotion=Locked
 }
 {
  bone1Name=dyn_01_hair_l
  bone2Name=Base_01_hair_l
  swing1LimitDegrees=45
  swing1Motion=Locked
  swing2LimitDegrees=45
  swing2Motion=Locked
  twistLimitDegrees=45
  twistMotion=Locked
 }
]
